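# Additive unlit Niagara sprite material for the Monarch emote's blast effect, animated entirely from particle relative time.
# A polygon gradient, with eased age driving its roundness, passes through gradient mask and reflection functions to form a moving band.
# A grayscale blast texture pans in radial coordinates, sampled with raw-UV derivatives for mip selection, then erodes against the polygon gradient.
# Band and erosion multiply into a powered, depth-faded intensity tinted by particle color and alpha.
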Material /SawbladeGun/Athena/Effects/Particles/Emotes/NS_Emote_Monarch/M_Emote_Monarch_Blast
graph {
 properties {bUsedWithNiagaraSprites=true, blendMode=BLEND_Additive, pixelDepthOffsetMode=PDOM_Legacy, shadingModel=MSM_Unlit, usedShadingModels=Unlit}
 nodes [
  n1 = OneMinus(None=ParticleRelativeTime())
  n2 = OneMinus(None=Multiply(A=@n1, B=@n1))
  n3 = ConstantBiasScale(None=@n2) {bias=-0.5, scale=2}
  n4 = ScalarParameter() {defaultValue=0.05, parameterName=Rotation}
  n5 = ComponentMask(None=OneMinus(None=MaterialFunctionCall!(UVs, Sides, Size, Roundness=@n3, Rotation=@n4, Normal Intensity) {materialFunction=/Game/Effects/Fort_Effects/MaterialFunctions/MF_Polygon}.Gradient)) {g=true}
  n6 = MaterialFunctionCall!(
   Gradient=MaterialFunctionCall!(Gradient=@n5, Mask Position=@n2, Mask Size=Constant() {r=1}) {materialFunction=/Game/Effects/Fort_Effects/MaterialFunctions/MF_GradientMask}.Linear
   Center Point=Constant() {r=0.8}
  ) {materialFunction=/Game/Effects/Fort_Effects/MaterialFunctions/MF_GradientReflected}
  n7 = Power(Base=OneMinus(None=ParticleRelativeTime()), Exp=ScalarParameter() {defaultValue=4, parameterName=Pan Exponent}) {constExponent=3}
  n8 = Constant()
  n9 = ScalarParameter() {defaultValue=-2.2, parameterName=Pan Amount}
  n10 = TextureCoordinate()
  n11 = TextureSample!(
   UVs=Panner(
    Coordinate=MaterialFunctionCall() {materialFunction=&VectorToRadialValue}.Radial Coordinates
    Time=OneMinus(None=@n7)
    Speed=AppendVector(A=@n8, B=@n9)
   ) {speedY=-3}
   Tex
   "DDX(UVs)"=DDX(Value=@n10)
   "DDY(UVs)"=DDY(Value=@n10)
   Apply View MipBias
  ) {mipValueMode=TMVM_Derivative, samplerType=SAMPLERTYPE_LinearGrayscale, texture=/Game/Effects/Fort_Effects/Textures/T_Emote_Monarch_Blast}
  n12 = OneMinus(None=@n11.R)
  n13 = MaterialFunctionCall(Vector 2=Subtract(A=@n10) {constB=0.5}) {materialFunction=/Engine/Functions/Engine_MaterialFunctions02/Utility/VectorLength}
  n14 = Power(Base=Saturate(None=Multiply(A=@n13.V2 Length) {constB=2})) {constExponent=0.333}
  n15 = OneMinus(None=Saturate(None=DotProduct(A=@n5, B=@n5)))
  n16 = MaterialFunctionCall!(
   Gradient=OneMinus(None=Multiply(A=@n12, B=@n14))
   Erosion Amount=OneMinus(None=SquareRoot(None=@n15))
   Softness=ScalarParameter() {defaultValue=1, parameterName=Softness, sliderMax=1}
  ) {materialFunction=/Game/Effects/Fort_Effects/MaterialFunctions/MF_Erosion}
  n17 = DepthFade(
   Opacity=Power(Base=Multiply(A=@n6.Result, B=@n16.Default Range), Exp=ScalarParameter() {defaultValue=3, parameterName=Exponent})
   FadeDistance=ScalarParameter() {defaultValue=11, parameterName=Depth Fade Distance}
  )
  n18 = ParticleColor()
  MaterialFunctionCall(Emissive=Multiply(A=Multiply(A=@n17, B=@n18.RGB), B=@n18.A)) {materialFunction=&MF_FX_EmissiveAdjustments}
 ]
}
