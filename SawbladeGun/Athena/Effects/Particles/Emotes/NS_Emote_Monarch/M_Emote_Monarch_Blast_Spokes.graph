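# Additive unlit Niagara sprite material that draws the blast spokes for the Monarch emote effect.
# A sine-based UV mask, narrowed by dividing by V, shapes a tapered spoke and multiplies a grayscale blast texture panned vertically with a per-particle random offset.
# An exponent parameter sharpens the result, which passes through depth fade, particle color and alpha, then emissive adjustments.
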Material /SawbladeGun/Athena/Effects/Particles/Emotes/NS_Emote_Monarch/M_Emote_Monarch_Blast_Spokes
graph {
 properties {bUsedWithNiagaraSprites=true, blendMode=BLEND_Additive, pixelDepthOffsetMode=PDOM_Legacy, shadingModel=MSM_Unlit, usedShadingModels=Unlit}
 nodes [
  n1 = TextureCoordinate()
  n2 = MaterialFunctionCall(Float2=Sine(None=Divide(A=@n1))) {materialFunction=&BreakOutFloat2Components}
  n3 = OneMinus(None=@n2.R)
  n4 = MaterialFunctionCall(Float2=@n1) {materialFunction=&BreakOutFloat2Components}
  n5 = Multiply(A=OneMinus(None=Divide(A=@n3, B=@n4.G)), B=@n2.G)
  n6 = TextureSample(UVs=Add(A=Panner() {speedY=1}, B=ParticleRandom())) {samplerType=SAMPLERTYPE_LinearGrayscale, texture=/Game/Effects/Fort_Effects/Textures/T_Emote_Monarch_Blast}
  n7 = DepthFade(
   Opacity=Power(Base=Multiply(A=@n5, B=@n6.R), Exp=ScalarParameter() {defaultValue=3, parameterName=Exponent})
   FadeDistance=ScalarParameter() {defaultValue=11, parameterName=Depth Fade Distance}
  )
  n8 = ParticleColor()
  MaterialFunctionCall(Emissive=Multiply(A=Multiply(A=@n7, B=@n8.RGB), B=@n8.A)) {materialFunction=&MF_FX_EmissiveAdjustments}
  n9 = OneMinus(None=@n5)
  MaterialFunctionCall!(Gradient=@n6.R, Erosion Amount=Multiply(A=@n9, B=@n9), Softness=Constant() {r=1}) {materialFunction=/Game/Effects/Fort_Effects/MaterialFunctions/MF_Erosion}
 ]
}
