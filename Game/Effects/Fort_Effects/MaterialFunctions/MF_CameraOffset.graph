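# Material function that pushes vertices toward or away from the camera through World Position Offset, optionally chained onto an incoming offset.
# The basic output scales the normalized vertex-to-camera vector by Camera Offset.
# The Safe output caps that push using distance to a near-plane point, found via a Custom node, weighted by view-direction alignment.
# An Opacity Multiplier fades as the cap shrinks the offset, hiding visible clipping with the Safe output.
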MaterialFunction /Game/Effects/Fort_Effects/MaterialFunctions/MF_CameraOffset
graph {nodes [
  n1 = CameraPositionWS()
  n2 = FunctionInput() {
   bUsePreviewValueAsDefault=true
   description="Additional WPO, optional, allows you to chain other WPO operations"
   inputName=Additional World Position Offset
  }
  n3 = Add(A=WorldPosition() {worldPositionShaderOffset=WPT_ExcludeAllShaderOffsets}.XYZ, B=@n2)
  n4 = Normalize(VectorInput=Subtract(A=@n1, B=@n3))
  n5 = FunctionInput() {
   bUsePreviewValueAsDefault=true
   description="How many units to push vertices towards the camera (negative values pushes them away)"
   inputName=Camera Offset
   inputType=FunctionInput_Scalar
  }
  FunctionOutput(None=Add(A=Multiply(A=@n4, B=@n5), B=@n2)) {
   description="Offsets vertices towards or away from camera. Might push them past the camera. If that's an issue, use Safe version instead."
   outputName=World Position Offset
  }
  n6 = MaterialFunctionCall() {materialFunction=&CameraDirectionVector}
  n7 = DotProduct(A=Multiply(A=@n6.Result) {constB=-1}, B=@n4)
  n8 = Custom() {
   code="(mul(float4(LWCToFloat(ResolvedView.WorldCameraOrigin).xyz,1),LWCToFloat(ResolvedView.WorldToClip))).z"
   desc="\n*** Original source before expression upgrade ***\n(mul(float4(ResolvedView.WorldCameraOrigin.xyz,1),View.WorldToClip)).z"
   description=Near Clip Plane
   outputType=CMOT_Float1
   inputs []
  }
  n9 = Subtract(A=Add(A=Multiply(A=@n6.Result, B=@n8), B=@n1), B=@n3)
  n10 = Min(
   A=Multiply(
    A=Saturate(None=@n7)
    B=MaterialFunctionCall(Vector 3=@n9) {materialFunction=/Engine/Functions/Engine_MaterialFunctions02/Utility/VectorLength}.V3 Length
   )
   B=@n5
  )
  FunctionOutput(None=Add(A=Multiply(A=@n10, B=@n4), B=@n2)) {
   description="Same as the regular version, but scales down effect when close to object. +10 instructions. Does nothing when used with negative values, i.e. offsetting away from camera."
   outputName=Safe World Position Offset
   sortPriority=1
  }
  n11 = Divide(A=Subtract(A=@n5, B=@n10) {constA=0}, B=@n5)
  FunctionOutput(None=OneMinus(None=Saturate(None=@n11))) {
   description="Fades to zero to prevent a visible clip when using the Safe version. Does nothing with the regular version. Can be passed through a vertex interpolator for some savings, since it does not rely on pixel information."
   outputName=Opacity Multiplier
   sortPriority=2
  }
 ]}
