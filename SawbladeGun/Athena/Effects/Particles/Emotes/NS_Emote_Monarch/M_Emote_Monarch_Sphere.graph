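# Additive, unlit Niagara sprite material for the Monarch emote effect, shading a flat particle as a glowing, textured sphere.
# MF_Sphere supplies UVs and height for three averaged panning texture samples, a soft edge mask and a sphere-shaped scene-depth fade.
# A reflected gradient shapes the textured glow, plus a center glow fading over particle life, tinted by particle color.
# Camera depth fade, refraction and camera offset complete the effect.
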Material /SawbladeGun/Athena/Effects/Particles/Emotes/NS_Emote_Monarch/M_Emote_Monarch_Sphere
graph {
 properties {
  bUsedWithNiagaraSprites=true
  blendMode=BLEND_Additive
  pixelDepthOffsetMode=PDOM_Legacy
  refractionMethod=RM_IndexOfRefraction
  shadingModel=MSM_Unlit
  usedShadingModels=Unlit
 }
 nodes [
  n1 = MaterialFunctionCall!(UV, Size, Curvature) {materialFunction=/Game/Effects/Fort_Effects/MaterialFunctions/MF_Sphere}
  n2 = Subtract(A=@n1.Height, B=ScalarParameter() {parameterName=OIpacity Offset, sliderMax=1})
  n3 = ScalarParameter() {defaultValue=0.9, parameterName=Opacity Softness}
  n4 = SceneDepth()
  n5 = Divide(A=ComponentMask(None=ParticleSize()) {r=true})
  n6 = Add(A=PixelDepth().R, B=Multiply(A=OneMinus(None=@n1.Height), B=@n5))
  n7 = ScalarParameter() {defaultValue=2, parameterName=Depth Falloff Fade Size}
  n8 = Saturate(None=Divide(A=Subtract(A=@n4.R, B=@n6), B=Multiply(A=@n5, B=@n7)))
  n9 = Multiply(A=Saturate(None=Divide(A=@n2, B=@n3)), B=@n8)
  n10 = MaterialFunctionCall!(
   UV=Multiply(A=@n1.UV, B=ScalarParameter() {defaultValue=2, parameterName=UV Scale})
   Rotation
   Speed=ScalarParameter() {defaultValue=0.4, parameterName=Speed}
   Time
   Rotate UVs
   Fractional Part
  ) {materialFunction=/Game/Effects/Fort_Effects/MaterialFunctions/MF_3UVPanner}
  n11 = TextureObject() {samplerType=SAMPLERTYPE_LinearGrayscale, texture=/Game/Effects/Fort_Effects/Textures/T_Emote_Monarch_Sphere}
  n12 = TextureSample(UVs=@n10.UV1, Tex=@n11) {samplerType=SAMPLERTYPE_LinearGrayscale}
  n13 = TextureSample(UVs=@n10.UV2, Tex=@n11) {samplerType=SAMPLERTYPE_LinearGrayscale}
  n14 = Divide(A=Add(A=Add(A=@n12.R, B=@n13.R), B=TextureSample(UVs=@n10.UV3, Tex=@n11) {samplerType=SAMPLERTYPE_LinearGrayscale}.R)) {constB=3}
  n15 = ParticleColor()
  n16 = Multiply(
   A=@n15.A
   B=MaterialFunctionCall(Fade Length=@n5, Fade Offset=@n5) {materialFunction=&CameraDepthFade}.Result
  )
  LinearInterpolate(B=ScalarParameter() {defaultValue=0.9, parameterName=Refraction}, Alpha=Multiply(A=Multiply(A=@n9, B=@n14), B=@n16)) {constA=1}
  n17 = MaterialFunctionCall!(Gradient=@n9, Center Point=ScalarParameter() {defaultValue=0.2, parameterName=Gradient Center}) {materialFunction=/Game/Effects/Fort_Effects/MaterialFunctions/MF_GradientReflected}
  n18 = Saturate(None=Divide(A=@n2, B=ScalarParameter() {defaultValue=0.1, parameterName=Center Color Softness, sliderMax=1}))
  n19 = OneMinus(None=ParticleRelativeTime())
  n20 = Multiply(A=ScalarParameter() {defaultValue=0.05, parameterName=Center Color Intensity}, B=Multiply(A=@n19, B=@n19))
  n21 = Add(
   A=Power(Base=Multiply(A=@n17.Result, B=@n14), Exp=ScalarParameter() {defaultValue=3, parameterName=Exponent, sliderMax=9})
   B=Multiply(A=Multiply(A=@n18, B=@n20), B=@n8)
  )
  MaterialFunctionCall(Emissive=Multiply(A=@n15.RGB, B=Multiply(A=@n21, B=@n16))) {materialFunction=&MF_FX_EmissiveAdjustments}
  MaterialFunctionCall!(Camera Offset=@n5, Additional World Position Offset) {materialFunction=/Game/Effects/Fort_Effects/MaterialFunctions/MF_CameraOffset}
  n22 = ParticleRelativeTime()
  OneMinus(None=Power(Base=OneMinus(None=@n22), Exp=ScalarParameter() {defaultValue=2, parameterName=Pan Exponent}) {constExponent=3})
  SquareRoot(None=@n22)
 ]
}
